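# Control Rig for the EagerUnifyHelmet female face accessory that drives a ponytail bone chain from body motion.
# Weighted parent constraints attach control nulls to head, spine, clavicle and pelvis bones, with head rotation blending in corrective offsets.
# Twist is distributed along the chain, then B-spline IK fits bones to controls, covering only bones 3–6 when rigid bodies are enabled without backbling.
# Construction adds helper nulls and optional FK controls.
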ControlRigBlueprint /BRCosmetics/Characters/Player/Female/Medium/Bodies/F_MED_EagerUnifyHelmet/Meshes/Parts/F_MED_EagerUnifyHelmet_FaceAcc_CtrlRig
graphs [
 CollapseNode_ContainedGraph {nodes [
   n1 = RigVMFunctionEntryNode()
   Constraints = RigVMCollapseNode(ExecutePin=@n1.ExecuteContext)
   RigVMFunction_Sequence = RigVMAggregateNode(ExecuteContext=@Constraints.ExecutePin)
   GetAlphaFromQuatDelta = RigVMFunctionReferenceNode(
    Rotation Order=ZYX
    child=(Type=Bone,Name="head")
    outputMinMax=(X=0,Y=1)
    parent=(Type=Null,Name="head_null")
    premult=1
    rotationMinMax=(X=0,Y=-60)
    triggerAxis=Z
    twistAxis=(X=1,Y=0,Z=0)
    ExecuteContext=@RigVMFunction_Sequence.A
   )
   > &RigUnit_ModifyTransforms(
    ItemToModify="((Item=(Type=Null,Name=\"dyn_pony_2_c_ctrl_null\"),Transform=(Rotation=(X=0,Y=0,Z=0,W=1),Translation=(X=-1.5,Y=0,Z=-1),Scale3D=(X=1,Y=1,Z=1))),(Item=(Type=Null,Name=\"dyn_pony_3_c_ctrl_null\"),Transform=(Rotation=(X=0,Y=0,Z=0,W=1),Translation=(X=2,Y=0,Z=4),Sca ...(truncated)"
    WeightMinimum=0
    Weight=@GetAlphaFromQuatDelta.alpha
   )
   joints_twist_c = RigVMDispatchNode(Value="((Type=Bone,Name=\"dyn_pony_3_c\"),(Type=Bone,Name=\"dyn_pony_4_c\"),(Type=Bone,Name=\"dyn_pony_5_c\"),(Type=Bone,Name=\"dyn_pony_6_c\"),(Type=Bone,Name=\"dyn_pony_7_c\"),(Type=Bone,Name=\"dyn_pony_8_c\"),(Type=Bone,Name=\"dyn_pony_9_c\"),(Type=Bone,Name=\"dyn_pony_10_c\" ...(truncated)") {template="DISPATCH_RigVMDispatch_Constant(io Value)"}
   TwistBones = RigVMFunctionReferenceNode(
    Axis to Twist=(bX=True,bY=true,bZ=true)
    Debug=false
    End=(Type=Bone,Name="neck_01")
    Start=(Type=Bone,Name="spine_02")
    ExecuteContext=@RigVMFunction_Sequence.B
    Bones to Twist=@joints_twist_c.Value
   )
   RigVMFunction_MathBoolNand2 = /Script/RigVM.RigVMFunction_MathBoolNand2(
    B=&RigVMFunction_MathBoolNot(Value=RigVMVariableNode() {variable=isWearingBackbling}.Value).Result
    A=RigVMVariableNode() {variable=pub isRigidBodies Enabled}.Value
   ) {is=struct RigVMUnitNode, template="Nand::Execute(in A,in B,out Result)"}
   RigVMFunction_ControlFlowBranch = &RigVMFunction_ControlFlowBranch(Condition=@RigVMFunction_MathBoolNand2.Result, ExecuteContext=@TwistBones.ExecuteContext)
   ItemArray = RigVMDispatchNode(Value="((Type=Control,Name=\"dyn_pony_2_c_ctrl\"),(Type=Control,Name=\"dyn_pony_3_c_ctrl\"),(Type=Control,Name=\"dyn_pony_4_c_ctrl\"),(Type=Control,Name=\"dyn_pony_5_c_ctrl\"),(Type=Control,Name=\"dyn_pony_6_c_ctrl\"),(Type=Control,Name=\"dyn_pony_7_c_ctrl\"),(Type=Control,N ...(truncated)") {template="DISPATCH_RigVMDispatch_Constant(io Value)"}
   ItemArray_1 = RigVMDispatchNode(Value="((Type=Bone,Name=\"dyn_pony_2_c\"),(Type=Bone,Name=\"dyn_pony_3_c\"),(Type=Bone,Name=\"dyn_pony_4_c\"),(Type=Bone,Name=\"dyn_pony_5_c\"),(Type=Bone,Name=\"dyn_pony_6_c\"),(Type=Bone,Name=\"dyn_pony_7_c\"),(Type=Bone,Name=\"dyn_pony_8_c\"),(Type=Bone,Name=\"dyn_pony_9_c\") ...(truncated)") {template="DISPATCH_RigVMDispatch_Constant(io Value)"}
   SplineIK_6 = RigVMFunctionReferenceNode(
    Primary Axis=(X=1,Y=0,Z=0)
    Samples Per Segment=1
    Secondary Spline Direction=(X=0,Y=0,Z=0)
    Spline Mode=BSpline
    Stretch=False
    Up Axis=(X=0,Y=0,Z=0)
    debug=false
    Controls=@ItemArray.Value
    Bones=@ItemArray_1.Value
    ExecuteContext=@RigVMFunction_ControlFlowBranch.True
   )
   ItemArray_2 = RigVMDispatchNode(Value=((Type=Bone,Name="dyn_pony_3_c"),(Type=Bone,Name="dyn_pony_4_c"),(Type=Bone,Name="dyn_pony_5_c"),(Type=Bone,Name="dyn_pony_6_c"))) {template="DISPATCH_RigVMDispatch_Constant(io Value)"}
   controls_ik_c_x4 = RigVMDispatchNode(Value=((Type=Control,Name="dyn_pony_3_c_ctrl"),(Type=Control,Name="dyn_pony_4_c_ctrl"),(Type=Control,Name="dyn_pony_5_c_ctrl"),(Type=Control,Name="dyn_pony_6_c_ctrl"))) {template="DISPATCH_RigVMDispatch_Constant(io Value)"}
   SplineIK = RigVMFunctionReferenceNode(
    Primary Axis=(X=1,Y=0,Z=0)
    Samples Per Segment=1
    Secondary Spline Direction=(X=0,Y=0,Z=0)
    Spline Mode=BSpline
    Stretch=False
    Up Axis=(X=0,Y=0,Z=0)
    debug=false
    Bones=@ItemArray_2.Value
    Controls=@controls_ik_c_x4.Value
    ExecuteContext=@RigVMFunction_ControlFlowBranch.False
   )
   RigVMFunctionReturnNode(ExecuteContext=@RigVMFunction_Sequence.C)
  ]}
 CollapseNode_ContainedGraph {nodes [
   n1 = RigVMFunctionEntryNode()
   RigVMFunction_Sequence = &RigVMFunction_Sequence(ExecuteContext=@n1.ExecutePin)
   &RigUnit_ParentConstraint(
    AdvancedSettings=(RotationOrderForFilter=XZY)
    Child=(Type=Bone,Name="dyn_pony_1_c")
    Parents=((Item=(Type=Bone,Name="head"),Weight=1))
    Weight=1
    bMaintainOffset=true
    ExecutePin=@RigVMFunction_Sequence.A
   )
   > &RigUnit_ParentConstraint(
    AdvancedSettings=(RotationOrderForFilter=XZY)
    Child=(Type=Null,Name="dyn_pony_2_c_ctrl_null")
    Parents=((Item=(Type=Bone,Name="head"),Weight=1))
    Weight=1
    bMaintainOffset=true
   )
   > &RigUnit_ParentConstraint(
    AdvancedSettings=(RotationOrderForFilter=XZY)
    Child=(Type=Null,Name="dyn_pony_3_c_ctrl_null")
    Parents=((Item=(Type=Bone,Name="head"),Weight=0.7),(Item=(Type=Bone,Name="neck_01"),Weight=0.1),(Item=(Type=Bone,Name="spine_05"),Weight=0.2))
    Weight=1
    bMaintainOffset=true
   )
   > &RigUnit_ParentConstraint(
    AdvancedSettings=(RotationOrderForFilter=XZY)
    Child=(Type=Null,Name="dyn_pony_4_c_ctrl_null")
    Parents=((Item=(Type=Bone,Name="spine_03"),Weight=0.0042),(Item=(Type=Bone,Name="neck_01"),Weight=0.1576),(Item=(Type=Bone,Name="head"),Weight=0.1784),(Item=(Type=Bone,Name="spine_05"),Weight=0.2528),(Item=(Type=Bone,Name="neck_02"),Weight=0.4071))
    Weight=1
   )
   > &RigUnit_ParentConstraint(
    AdvancedSettings=(RotationOrderForFilter=XZY)
    Child=(Type=Null,Name="dyn_pony_5_c_ctrl_null")
    Parents="((Item=(Type=Bone,Name=\"spine_04\"),Weight=0.0001),(Item=(Type=Bone,Name=\"head\"),Weight=0.0008),(Item=(Type=Bone,Name=\"neck_02\"),Weight=0.0052),(Item=(Type=Bone,Name=\"spine_03\"),Weight=0.0099),(Item=(Type=Bone,Name=\"neck_01\"),Weight=0.0108),(Item=(Type=Bone ...(truncated)"
    Weight=1
   )
   > &RigUnit_ParentConstraint(
    AdvancedSettings=(RotationOrderForFilter=XZY)
    Child=(Type=Null,Name="dyn_pony_6_c_ctrl_null")
    Parents="((Item=(Type=Bone,Name=\"spine_03\"),Weight=0.0002),(Item=(Type=Bone,Name=\"neck_01\"),Weight=0.0002),(Item=(Type=Bone,Name=\"spine_04\"),Weight=0.03),(Item=(Type=Bone,Name=\"clavicle_r\"),Weight=0.1708),(Item=(Type=Bone,Name=\"clavicle_l\"),Weight=0.1735),(Item=(Ty ...(truncated)"
    Weight=1
   )
   > &RigUnit_ParentConstraint(
    AdvancedSettings=(RotationOrderForFilter=XZY)
    Child=(Type=Null,Name="dyn_pony_7_c_ctrl_null")
    Parents="((Item=(Type=Bone,Name=\"pelvis\"),Weight=0.0035),(Item=(Type=Bone,Name=\"spine_01\"),Weight=0.037),(Item=(Type=Bone,Name=\"clavicle_l\"),Weight=0.0829),(Item=(Type=Bone,Name=\"clavicle_r\"),Weight=0.0831),(Item=(Type=Bone,Name=\"spine_03\"),Weight=0.1055),(Item=(Ty ...(truncated)"
    Weight=1
   )
   > &RigUnit_ParentConstraint(
    AdvancedSettings=(RotationOrderForFilter=XZY)
    Child=(Type=Null,Name="dyn_pony_8_c_ctrl_null")
    Parents="((Item=(Type=Bone,Name=\"pelvis\"),Weight=0.013),(Item=(Type=Bone,Name=\"spine_01\"),Weight=0.0551),(Item=(Type=Bone,Name=\"clavicle_l\"),Weight=0.0591),(Item=(Type=Bone,Name=\"clavicle_r\"),Weight=0.0591),(Item=(Type=Bone,Name=\"spine_03\"),Weight=0.1413),(Item=(Ty ...(truncated)"
    Weight=1
   )
   > &RigUnit_ParentConstraint(
    AdvancedSettings=(RotationOrderForFilter=XZY)
    Child=(Type=Null,Name="dyn_pony_9_c_ctrl_null")
    Parents="((Item=(Type=Bone,Name=\"spine_05\"),Weight=0.0176),(Item=(Type=Bone,Name=\"spine_01\"),Weight=0.107),(Item=(Type=Bone,Name=\"pelvis\"),Weight=0.1298),(Item=(Type=Bone,Name=\"spine_03\"),Weight=0.1376),(Item=(Type=Bone,Name=\"spine_04\"),Weight=0.1505),(Item=(Type=B ...(truncated)"
    Weight=1
   )
   > &RigUnit_ParentConstraint(
    AdvancedSettings=(RotationOrderForFilter=XZY)
    Child=(Type=Null,Name="dyn_pony_10_c_ctrl_null")
    Parents=((Item=(Type=Bone,Name="spine_01"),Weight=0.014),(Item=(Type=Bone,Name="spine_02"),Weight=0.0223),(Item=(Type=Bone,Name="pelvis"),Weight=0.9637))
    Weight=1
   )
   RigVMFunctionReturnNode(ExecutePin=@RigVMFunction_Sequence.B)
  ]}
 New Function_ContainedGraph {nodes [
   n1 = RigVMFunctionEntryNode()
   RigVMFunction_MathBoolToInteger = &RigVMFunction_MathBoolToInteger(Value=@n1.onlyTwist)
   RigVMFunction_MathQuaternionSwingTwist = &RigVMFunction_MathQuaternionSwingTwist(
    twistAxis=@n1.twistAxis
    Input=&RigUnit_GetRelativeTransformForItem(bChildInitial=False, bParentInitial=False, child=@n1.child, parent=@n1.parent).RelativeTransform.Rotation
   )
   n2 = RigVMDispatchNode(
    Values=((),())
    Index=@RigVMFunction_MathBoolToInteger.Result
    Values.0=@RigVMFunction_MathQuaternionSwingTwist.Swing
    Values.1=@RigVMFunction_MathQuaternionSwingTwist.Twist
   ) {template="DISPATCH_RigVMDispatch_SelectInt32(in Index,in Values,out Result)"}
   RigVMFunction_MathQuaternionToEuler = &RigVMFunction_MathQuaternionToEuler(Value=@n2.Result, RotationOrder=@n1.Rotation Order)
   > DISPATCH_RigVMDispatch_Print = RigVMDispatchNode(ScreenColor=(R=1,G=1,B=1,A=1), ScreenDuration=1.5, Enabled=@n1.debugRotation, Value=@RigVMFunction_MathQuaternionToEuler.Result) {template="DISPATCH_RigVMDispatch_Print(in Prefix,in Value,in Enabled,in ScreenDuration,in ScreenColor)"}
   Cast = RigVMDispatchNode(Value=@n1.triggerAxis) {template="DISPATCH_RigVMDispatch_CastEnumToInt(in Value,out Result)"}
   DISPATCH_RigVMDispatch_SelectInt32 = RigVMDispatchNode(
    Values=(0,0,0)
    Index=@Cast.Result
    Values.0=@RigVMFunction_MathQuaternionToEuler.Result.X
    Values.1=@RigVMFunction_MathQuaternionToEuler.Result.Y
    Values.2=@RigVMFunction_MathQuaternionToEuler.Result.Z
   ) {template="DISPATCH_RigVMDispatch_SelectInt32(in Index,in Values,out Result)"}
   Remap = &RigVMFunction_MathDoubleRemap(
    bClamp=False
    Value=@DISPATCH_RigVMDispatch_SelectInt32.Result
    TargetMinimum=@n1.outputMinMax.X
    TargetMaximum=@n1.outputMinMax.Y
    SourceMinimum=@n1.rotationMinMax.X
    SourceMaximum=@n1.rotationMinMax.Y
   )
   Absolute = &RigVMFunction_MathDoubleAbs(Value=@Remap.Result)
   Multiply = &RigVMFunction_MathDoubleMul(
    A=&RigVMFunction_MathDoubleClamp(Value=RigVMDispatchNode(True=@Absolute.Result, Condition=@n1.absoluteValue, False=@Remap.Result) {template="DISPATCH_RigVMDispatch_If(in Condition,in True,in False,out Result)"}.Result).Result
    B=@n1.premult
   )
   > RigVMFunctionReturnNode(outputQuat=@n2.Result, outputWeight=@Multiply.Result)
  ]}
 New Function_ContainedGraph {nodes [
   RigVMFunctionEntryNode()
   RigVMFunction_MathBoolToInteger = &RigVMFunction_MathBoolToInteger(Value=RigVMVariableNode() {variable=onlyTwist}.Value)
   RigVMFunction_MathQuaternionSwingTwist = &RigVMFunction_MathQuaternionSwingTwist(
    Input=&RigUnit_GetRelativeTransformForItem(
     bChildInitial=False
     bParentInitial=False
     Child=RigVMVariableNode() {variable=child}.Value
     Parent=RigVMVariableNode() {variable=parent}.Value
    ).RelativeTransform.Rotation
    TwistAxis=RigVMVariableNode() {variable=twistAxis}.Value
   )
   n1 = RigVMDispatchNode(
    Values=((),())
    Index=@RigVMFunction_MathBoolToInteger.Result
    Values.0=@RigVMFunction_MathQuaternionSwingTwist.Swing
    Values.1=@RigVMFunction_MathQuaternionSwingTwist.Twist
   ) {template="DISPATCH_RigVMDispatch_SelectInt32(in Index,in Values,out Result)"}
   > RigVMVariableNode(Value=@n1.Result) {variable=entry_outputQuat}
   RigVMFunction_MathQuaternionToEuler = &RigVMFunction_MathQuaternionToEuler(Value=@n1.Result, RotationOrder=RigVMVariableNode() {variable=Rotation Order}.Value)
   > DISPATCH_RigVMDispatch_Print = RigVMDispatchNode(
    ScreenColor=(R=1,G=1,B=1,A=1)
    ScreenDuration=1.5
    Value=@RigVMFunction_MathQuaternionToEuler.Result
    Enabled=RigVMVariableNode() {variable=debugRotation}.Value
   ) {template="DISPATCH_RigVMDispatch_Print(in Prefix,in Value,in Enabled,in ScreenDuration,in ScreenColor)"}
   Cast = RigVMDispatchNode(Value=RigVMVariableNode() {variable=triggerAxis}.Value) {template="DISPATCH_RigVMDispatch_CastEnumToInt(in Value,out Result)"}
   DISPATCH_RigVMDispatch_SelectInt32 = RigVMDispatchNode(
    Values=(0,0,0)
    Index=@Cast.Result
    Values.0=@RigVMFunction_MathQuaternionToEuler.Result.X
    Values.1=@RigVMFunction_MathQuaternionToEuler.Result.Y
    Values.2=@RigVMFunction_MathQuaternionToEuler.Result.Z
   ) {template="DISPATCH_RigVMDispatch_SelectInt32(in Index,in Values,out Result)"}
   n2 = RigVMVariableNode() {variable=rotationMinMax}
   n3 = RigVMVariableNode() {variable=outputMinMax}
   Remap = &RigVMFunction_MathDoubleRemap(
    bClamp=False
    Value=@DISPATCH_RigVMDispatch_SelectInt32.Result
    SourceMinimum=@n2.Value.X
    SourceMaximum=@n2.Value.Y
    TargetMinimum=@n3.Value.X
    TargetMaximum=@n3.Value.Y
   )
   Absolute = &RigVMFunction_MathDoubleAbs(Value=@Remap.Result)
   n4 = RigVMVariableNode() {variable=absoluteValue}
   Multiply = &RigVMFunction_MathDoubleMul(
    A=&RigVMFunction_MathDoubleClamp(Value=RigVMDispatchNode(True=@Absolute.Result, False=@Remap.Result, Condition=@n4.Value) {template="DISPATCH_RigVMDispatch_If(in Condition,in True,in False,out Result)"}.Result).Result
    B=RigVMVariableNode() {variable=premult}.Value
   )
   > RigVMFunctionReturnNode(alpha=@Multiply.Result, rawQuat=RigVMVariableNode() {variable=entry_outputQuat}.Value)
  ]}
 RigVMFunctionLibrary {nodes [
   DeltaRotation v4 = RigVMCollapseNode(
    ExecuteContext=()
    Rotation Order=ZYX
    child=(Type=Bone,Name="None")
    outputMinMax=(X=0,Y=1)
    parent=(Type=Bone,Name="None")
    premult=1
    rotationMinMax=(X=0,Y=0)
    twistAxis=(X=1,Y=0,Z=0)
   )
   GetAlphaFromQuatDelta = RigVMCollapseNode(
    ExecuteContext=()
    Rotation Order=ZYX
    child=(Type=Bone,Name="None")
    outputMinMax=(X=0,Y=1)
    parent=(Type=Bone,Name="None")
    premult=1
    rotationMinMax=(X=0,Y=0)
    twistAxis=(X=1,Y=0,Z=0)
   )
  ]}
 RigVMFunction_Sequence_1_ContainedGraph {nodes [
   n1 = RigVMFunctionEntryNode()
   RigVMFunction_Sequence = &RigVMFunction_Sequence(ExecuteContext=@n1.ExecuteContext)
   n2 = &RigVMFunction_Sequence(ExecuteContext=@RigVMFunction_Sequence.B)
   RigVMFunctionReturnNode(A=@RigVMFunction_Sequence.A, B=@n2.A, C=@n2.B)
  ]}
 RigVMModel {nodes [
   RigUnit_BeginExecution = &RigUnit_BeginExecution()
   setFKCtrl = RigVMFunctionReferenceNode(ExecuteContext=@RigUnit_BeginExecution.ExecutePin, setFK_Ctrl=RigVMVariableNode() {variable="fkControls "}.Value)
   > Hair = RigVMCollapseNode()
  ]}
 RigVMModel Construction Event Graph {nodes [
   n1 = &RigUnit_PrepareForExecution()
   RigVMFunction_ControlFlowBranch = &RigVMFunction_ControlFlowBranch(ExecuteContext=@n1.ExecutePin, Condition=RigVMVariableNode() {variable="fkControls "}.Value)
   Item Array = RigVMDispatchNode(Value="((Type=Bone,Name=\"thigh_l\"),(Type=Bone,Name=\"calf_l\"),(Type=Bone,Name=\"foot_l\"),(Type=Bone,Name=\"thigh_r\"),(Type=Bone,Name=\"calf_r\"),(Type=Bone,Name=\"foot_r\"),(Type=Bone,Name=\"clavicle_l\"),(Type=Bone,Name=\"upperarm_l\"),(Type=Bone,Name=\"hand_l\"),(Type=Bone, ...(truncated)") {template="DISPATCH_RigVMDispatch_Constant(io Value)"}
   addNullObject = RigVMFunctionReferenceNode(
    Options_Null="(pelvis_1_#1=False,spine_3_#2=true,clavicle_5_#3=false,arm_7_#4=false,leg_9_#5=False,head_11_46DC15EF465CD ...(truncated)"
    Custom_Joints=@"Item Array".Value
    ExecuteContext=@RigVMFunction_ControlFlowBranch.Completed
   )
   > &RigUnit_HierarchyAddNull(
    Name=neck_02_null
    Parent=(Type=Bone,Name="spine_05")
    Space=LocalSpace
    Transform=&RigUnit_GetRelativeTransformForItem(Child=(Type=Bone,Name="neck_02"), Parent=(Type=Bone,Name="spine_05"), bChildInitial=true, bParentInitial=true).RelativeTransform
   )
   addFKCtrl = RigVMFunctionReferenceNode(
    Custom_FKCtrl=()
    Options_FKCtrl="(pelvis_1_#1=true,spine_3_#2=true,clavicle_5_#3=true,arm_7_#4=true,leg_9_#5=true,head_11_46DC15EF465CD2979 ...(truncated)"
    Shape_Settings="(pelvis_2_#6=(bVisible=True,Name=\"Hexagon_Thin\",Color=(R=0.0117,G=0.8281,B=1,A=1),Transform=(Rotation=(X=0,Y=-0.7071,Z=0,W=0.7071),Translation=(X=2,Y=0,Z=0),Scale3D=(X=5,Y=5,Z=7))),spine_4_#7=(bVi ...(truncated)"
    ExecuteContext=@RigVMFunction_ControlFlowBranch.True
   )
  ]}
]
hierarchy [
 root {loc=<0 0 0>, rot=<0 0 0 1>}
 pelvis {parent=root, loc=<0 -0.5301 98.0571>, rot=<0.0127 0.707 -0.0127 -0.707>}
 spine_01 {parent=pelvis, loc=<6.1208 -0.9328 0>, rot=<0 0 0.034 -0.9994>}
 spine_02 {parent=spine_01, loc=<6.5997 -0.2095 0>, rot=<0 0 0.0238 -0.9997>}
 spine_03 {parent=spine_02, loc=<6.4914 -0.1237 0>, rot=<0 0 0.0149 -0.9999>}
 spine_04 {parent=spine_03, loc=<6.9598 0.0794 0>, rot=<0 0 -0.0132 -0.9999>}
 spine_05 {parent=spine_04, loc=<6.9987 0 0>, rot=<0 0 -0.0015 -1>}
 neck_01 {parent=spine_05, loc=<11.3138 0.1529 0>, rot=<0 0 -0.2003 -0.9797>}
 neck_02 {parent=neck_01, loc=<4.5627 0 0>, rot=<0 0 0.0504 -0.9987>}
 head {parent=neck_02, loc=<5.3914 0 0>, rot=<0 0 0.115 -0.9934>}
 base_pony_1_c {parent=head, loc=<8.8073 -8.4498 0>, rot=<0.5412 -0.4551 0.4551 -0.5412>}
 dyn_pony_1_c {parent=base_pony_1_c, loc=<0 0 0>, rot=<0 0 0 1>}
 dyn_pony_2_c {parent=dyn_pony_1_c, loc=<3.9895 0 0>, rot=<0 -0.4942 0 -0.8694>}
 dyn_pony_3_c {parent=dyn_pony_2_c, loc=<4.3292 0 0>, rot=<0 -0.2816 0 -0.9595>}
 dyn_pony_4_c {parent=dyn_pony_3_c, loc=<7.5 0 0>, rot=<0 -0.0664 0 -0.9978>}
 dyn_pony_5_c {parent=dyn_pony_4_c, loc=<7.5 0 0>, rot=<0 -0.001 0 -1>}
 dyn_pony_6_c {parent=dyn_pony_5_c, loc=<7.5 0 0>, rot=<0 -0.0149 0 -0.9999>}
 dyn_pony_7_c {parent=dyn_pony_6_c, loc=<7.5 0 0>, rot=<0 0 0 1>}
 dyn_pony_8_c {parent=dyn_pony_7_c, loc=<7.5 0 0>, rot=<0 0.018 0 -0.9998>}
 dyn_pony_9_c {parent=dyn_pony_8_c, loc=<7.5 0 0>, rot=<0 0.2095 0 -0.9778>}
 dyn_pony_10_c {parent=dyn_pony_9_c, loc=<8 0 0>, rot=<0 0 0 1>}
 clavicle_l {parent=spine_05, loc=<5.4831 -0.1053 -3.2946>, rot=<0.0465 -0.7289 -0.009 -0.6829>}
 upperarm_l {parent=clavicle_l, loc=<9.8658 -1.2136 -0.4105>, rot=<0.0164 -0.3245 -0.0269 -0.9454>}
 upperarm_twist_01_l {parent=upperarm_l, loc=<7.2135 -0.1134 0>, rot=<0 0 0 1>}
 lowerarm_l {parent=upperarm_l, loc=<26.465 -0.8345 0.259>, rot=<0.0036 0.0161 -0.2298 -0.9731>}
 upperarm_twist_02_l {parent=upperarm_l, loc=<19.7998 -0.1134 0>, rot=<0 0 0 1>}
 clavicle_r {parent=spine_05, loc=<5.4831 -0.1053 3.2946>, rot=<0.7289 0.0465 -0.6829 0.009>}
 upperarm_r {parent=clavicle_r, loc=<-9.8658 1.2136 0.4105>, rot=<0.0164 -0.3245 -0.0269 -0.9454>}
 upperarm_twist_01_r {parent=upperarm_r, loc=<-7.2135 0.1134 0>, rot=<0 0 0 -1>}
 lowerarm_r {parent=upperarm_r, loc=<-26.465 0.8345 -0.259>, rot=<0.0036 0.0161 -0.2298 -0.9731>}
 upperarm_twist_02_r {parent=upperarm_r, loc=<-19.7998 0.1134 0>, rot=<0 0 0 1>}
 dyn_pony_1_c_ctrl_null {type=Null, loc=<0 -10.2097 160.7714>, rot=<0.0579 0.0579 0.7047 -0.7047>}
 dyn_pony_1_c_ctrl {controlType=EULER_TRANSFORM, parent=dyn_pony_1_c_ctrl_null, type=Control, loc=<-36.3239 0 -156.9467>, rot=<-0.0579 -0.0579 -0.7047 -0.7047>}
 dyn_pony_2_c_ctrl_null {type=Null, loc=<0 -14.1456 161.4229>, rot=<-0.2979 -0.2979 0.6413 -0.6413>}
 dyn_pony_2_c_ctrl {controlType=EULER_TRANSFORM, parent=dyn_pony_2_c_ctrl_null, type=Control, loc=<-2.041 0 -3.4279>, rot=<0 0.4942 0 -0.8694>}
 dyn_pony_3_c_ctrl_null {type=Null, loc=<0 -16.9381 158.1147>, rot=<-0.4664 -0.4664 0.5314 -0.5314>}
 dyn_pony_3_c_ctrl {controlType=EULER_TRANSFORM, parent=dyn_pony_3_c_ctrl_null, type=Control, loc=<-3.6425 0 -2.3397>, rot=<0 0.2816 0 -0.9595>}
 dyn_pony_4_c_ctrl_null {type=Null, loc=<0 -17.9111 150.678>, rot=<-0.5007 -0.5007 0.4993 -0.4993>}
 dyn_pony_4_c_ctrl {controlType=EULER_TRANSFORM, parent=dyn_pony_4_c_ctrl_null, type=Control, loc=<-7.4338 0 -0.9943>, rot=<0 0.0664 0 -0.9978>}
 dyn_pony_5_c_ctrl_null {type=Null, loc=<0 -17.8896 143.1781>, rot=<-0.5012 -0.5012 0.4988 -0.4988>}
 dyn_pony_5_c_ctrl {controlType=EULER_TRANSFORM, parent=dyn_pony_5_c_ctrl_null, type=Control, loc=<-7.5 0 -0.0148>, rot=<0 0.001 0 -1>}
 dyn_pony_6_c_ctrl_null {type=Null, loc=<0 -17.8533 135.6782>, rot=<-0.5086 -0.5086 0.4913 -0.4913>}
 dyn_pony_6_c_ctrl {controlType=EULER_TRANSFORM, parent=dyn_pony_6_c_ctrl_null, type=Control, loc=<-7.4967 0 -0.2235>, rot=<0 0.0149 0 -0.9999>}
 dyn_pony_7_c_ctrl_null {type=Null, loc=<0 -17.5934 128.1827>, rot=<-0.5086 -0.5086 0.4913 -0.4913>}
 dyn_pony_7_c_ctrl {controlType=EULER_TRANSFORM, parent=dyn_pony_7_c_ctrl_null, type=Control, loc=<-7.5 0 0>, rot=<0 0 0 1>}
 dyn_pony_8_c_ctrl_null {type=Null, loc=<0 -17.3336 120.6872>, rot=<-0.4997 -0.4997 0.5003 -0.5003>}
 dyn_pony_8_c_ctrl {controlType=EULER_TRANSFORM, parent=dyn_pony_8_c_ctrl_null, type=Control, loc=<-7.4951 0 0.27>, rot=<0 -0.018 0 -0.9998>}
 dyn_pony_9_c_ctrl_null {type=Null, loc=<0 -17.3438 113.1872>, rot=<-0.3837 -0.3837 0.5939 -0.5939>}
 dyn_pony_9_c_ctrl {controlType=EULER_TRANSFORM, parent=dyn_pony_9_c_ctrl_null, type=Control, loc=<-6.8414 0 3.0734>, rot=<0 -0.2095 0 -0.9778>}
 dyn_pony_10_c_ctrl_null {type=Null, loc=<0 -20.632 105.8942>, rot=<-0.3837 -0.3837 0.5939 -0.5939>}
 dyn_pony_10_c_ctrl {controlType=EULER_TRANSFORM, parent=dyn_pony_10_c_ctrl_null, type=Control, loc=<-8 0 0>, rot=<0 0 0 1>}
]
variables [
 "fkControls " {type=bool, default=False}
 isWearingBackbling {type=bool, default=False}
 pub isRigidBodies Enabled {type=bool, default=False}
]
