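# Physics asset for the FaceAcc part of the female medium EagerUnifyHelmet cosmetic body, setting up a ponytail bone chain with body collision shapes.
# Nine dyn_pony capsules link in a chain with tight angular limits, with twist locked on the two joints nearest the root.
# The chain hangs from base_pony_1_c, which is angularly locked to the head.
# Capsules on head, neck, spine, pelvis, clavicles and arms form the body shapes, with arm joints angularly locked.
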PhysicsAsset /BRCosmetics/Characters/Player/Female/Medium/Bodies/F_MED_EagerUnifyHelmet/Meshes/Parts/F_MED_EagerUnifyHelmet_FaceAcc_Physics
bodies {
 base_pony_1_c [{shapeType=Capsule, length=0.2992, radius=0.8808, shapeName=base_pony_1_c_capsule}]
 clavicle_l [{
   shapeType=Capsule
   length=23.0848
   radius=8.9012
   shapeName=clavicle_l_capsule
   center=<17.2561 -0.9751 -1.6716>
   rotation=<rot 87.884 169.2103 170.7161>
  }]
 clavicle_r [{
   shapeType=Capsule
   length=23.0848
   radius=8.9012
   shapeName=clavicle_r_capsule
   center=<-17.2561 0.9751 1.6716>
   rotation=<rot 87.884 169.2103 -9.2839>
  }]
 dyn_pony_1_c [{shapeType=Capsule, length=1.1413, radius=2.4569, shapeName=dyn_pony_1_c_capsule, center=<1.9667 0 0>, rotation=<rot 0 0 90>}]
 dyn_pony_2_c [{shapeType=Capsule, length=1.3415, radius=3.5983, shapeName=dyn_pony_2_c_capsule, center=<2.5775 0 0.3722>, rotation=<rot 0 0 90>}]
 dyn_pony_3_c [{shapeType=Capsule, length=1.7701, radius=3.5983, shapeName=dyn_pony_3_c_capsule, center=<3.9248 0.2604 0.5539>, rotation=<rot 0 0 90>}]
 dyn_pony_4_c [{shapeType=Capsule, length=3.6021, radius=3.9782, shapeName=dyn_pony_4_c_capsule, center=<3.9248 0 0>, rotation=<rot 0 0 90>}]
 dyn_pony_5_c [{shapeType=Capsule, length=6.4195, radius=4.2784, shapeName=dyn_pony_5_c_capsule, center=<3.9248 -0.2642 0>, rotation=<rot 0 0 90>}]
 dyn_pony_6_c [{shapeType=Capsule, length=6.4752, radius=4.8529, shapeName=dyn_pony_6_c_capsule, center=<3.9248 -0.419 0>, rotation=<rot 0 0 90>}]
 dyn_pony_7_c [{shapeType=Capsule, length=8.6748, radius=4.2841, shapeName=dyn_pony_7_c_capsule, center=<3.9248 -0.6646 0>, rotation=<rot 0 0 90>}]
 dyn_pony_8_c [{shapeType=Capsule, length=7.8641, radius=3.5983, shapeName=dyn_pony_8_c_capsule, center=<3.9248 -0.6967 0>, rotation=<rot 0 0 90>}]
 dyn_pony_9_c [{shapeType=Capsule, length=3.9393, radius=3.5983, shapeName=dyn_pony_9_c_capsule, center=<3.9248 -1.182 0>, rotation=<rot 0 0 90>}]
 head [{shapeType=Capsule, length=37.6339, radius=11.3477, center=<2.824 1.4463 -0.0008>}]
 lowerarm_l [{
   shapeType=Capsule
   length=26.2566
   radius=8.3523
   shapeName=lowerarm_l_capsule
   center=<21.0352 0 -1.8607>
   rotation=<rot -0.0001 -90 96.037>
  }]
 lowerarm_r [{
   shapeType=Capsule
   length=26.2566
   radius=8.3523
   shapeName=lowerarm_r_capsule
   center=<-21.0352 0 1.8607>
   rotation=<rot -0.0001 -90 -83.963>
  }]
 neck_01 [{shapeType=Capsule, length=1.4368, radius=6.1234, shapeName=neck_01_capsule, center=<1.3406 0.88 0>, rotation=<rot 90 0 0>}]
 pelvis [{shapeType=Capsule, length=18.1954, radius=13.3885, shapeName=pelvis_capsule, center=<-5.087 -1.9631 0>, rotation=<rot 0 -90 0>}]
 spine_02 [{shapeType=Capsule, length=9.7448, radius=9.255, shapeName=spine_02_capsule, center=<0 0.7644 0>, rotation=<rot 0 -90 0>}]
 spine_03 [{shapeType=Capsule, length=10.9278, radius=10.0521, shapeName=spine_02_capsule, center=<2.5607 0.8486 0>, rotation=<rot 0 -90 0>}]
 spine_05 [{shapeType=Capsule, length=12.25, radius=12.5, center=<-0.8242 3.3872 -0.0002>}]
 upperarm_l [{shapeType=Capsule, length=19.6851, radius=8.4847, shapeName=upperarm_l_capsule, center=<15.3093 0 0>, rotation=<rot 0 -90 90>}]
 upperarm_r [{shapeType=Capsule, length=19.6851, radius=8.4847, shapeName=upperarm_r_capsule, center=<-15.3093 0 0>, rotation=<rot 0 -90 -90>}]
}
constraints [
 {bone1Name=dyn_pony_9_c, bone2Name=dyn_pony_8_c, swing1LimitDegrees=10, swing2LimitDegrees=10, twistLimitDegrees=10}
 {bone1Name=dyn_pony_8_c, bone2Name=dyn_pony_7_c, swing1LimitDegrees=10, swing2LimitDegrees=10, twistLimitDegrees=10}
 {bone1Name=dyn_pony_7_c, bone2Name=dyn_pony_6_c, swing1LimitDegrees=10, swing2LimitDegrees=10, twistLimitDegrees=10}
 {bone1Name=dyn_pony_6_c, bone2Name=dyn_pony_5_c, swing1LimitDegrees=10, swing2LimitDegrees=10, twistLimitDegrees=10}
 {bone1Name=dyn_pony_5_c, bone2Name=dyn_pony_4_c, swing1LimitDegrees=10, swing2LimitDegrees=10, twistLimitDegrees=10}
 {bone1Name=dyn_pony_4_c, bone2Name=dyn_pony_3_c, swing1LimitDegrees=10, swing2LimitDegrees=10, twistLimitDegrees=10}
 {bone1Name=dyn_pony_3_c, bone2Name=dyn_pony_2_c, swing1LimitDegrees=10, swing2LimitDegrees=10, twistLimitDegrees=10}
 {bone1Name=dyn_pony_2_c, bone2Name=dyn_pony_1_c, swing1LimitDegrees=15, swing2LimitDegrees=15, twistLimitDegrees=45, twistMotion=Locked}
 {bone1Name=dyn_pony_1_c, bone2Name=base_pony_1_c, swing1LimitDegrees=5, swing2LimitDegrees=5, twistLimitDegrees=45, twistMotion=Locked}
 {
  bone1Name=base_pony_1_c
  bone2Name=head
  swing1LimitDegrees=45
  swing1Motion=Locked
  swing2LimitDegrees=45
  swing2Motion=Locked
  twistLimitDegrees=45
  twistMotion=Locked
 }
 {
  bone1Name=upperarm_l
  bone2Name=clavicle_l
  swing1LimitDegrees=45
  swing1Motion=Locked
  swing2LimitDegrees=45
  swing2Motion=Locked
  twistLimitDegrees=45
  twistMotion=Locked
 }
 {
  bone1Name=lowerarm_l
  bone2Name=upperarm_l
  swing1LimitDegrees=45
  swing1Motion=Locked
  swing2LimitDegrees=45
  swing2Motion=Locked
  twistLimitDegrees=45
  twistMotion=Locked
 }
 {
  bone1Name=upperarm_r
  bone2Name=clavicle_r
  swing1LimitDegrees=45
  swing1Motion=Locked
  swing2LimitDegrees=45
  swing2Motion=Locked
  twistLimitDegrees=45
  twistMotion=Locked
 }
 {
  bone1Name=lowerarm_r
  bone2Name=upperarm_r
  swing1LimitDegrees=45
  swing1Motion=Locked
  swing2LimitDegrees=45
  swing2Motion=Locked
  twistLimitDegrees=45
  twistMotion=Locked
 }
]
