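# Anim Blueprint for the EagerUnify helmet's face accessory that adds secondary motion to its ponytail bones.
# The copied parent pose runs through a Control Rig into a cache feeding RigidBody and AnimDynamics branches.
# RBanActive, set when DBNO or wearing non-cape backbling, selects stiff pelvis-relative AnimDynamics; otherwise a nested blend with an unwired selector applies, before final bone offsets.
# Updates also compute gravity override and idle-frontend state, neither visibly read by the graph.
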AnimBlueprint /BRCosmetics/Characters/Player/Female/Medium/Bodies/F_MED_EagerUnifyHelmet/Meshes/Parts/F_MED_EagerUnifyHelmet_FaceAcc_AnimBP
target_skeleton=&/F_MED_EagerUnifyHelmet_FaceAcc_Skeleton
graphs [
 AnimGraph {nodes [
   n1 = RigidBody!(ComponentPose=Local To Component(LocalPose=Use cached pose 'InputCache'()), Alpha, OverrideWorldGravity:Vector="0,0,0") -> (Pose) {
    bClampLinearTranslationLimitToRefPose=true
    bOverrideWorldGravity=true
    overridePhysicsAsset=&/F_MED_EagerUnifyHelmet_FaceAcc_Physics
    simulationSpace=BaseBoneSpace
    alphaBoolBlend {blendOption=Linear}
    alphaScaleBiasClamp {outRange {max=0.4, min=1}}
    baseBoneRef {boneName=neck_01}
    componentAppliedLinearAccClamp=<3000 3000 3000>
    componentLinearAccScale=<0.4 0.35 0.35>
    componentLinearVelScale=<0.4 0.35 0.35>
    overrideWorldGravity {z=-980}
   }
   n2 = "Transform (Modify) Bone - Bone: dyn_pony_3_c"!(ComponentPose=@n1, bAlphaBoolEnabled:bool=True, Translation="0,0,0", Rotation="0,0,35") -> (Pose) {
    alphaInputType=Bool
    rotationMode=BMM_Additive
    translationMode=BMM_Additive
    alphaBoolBlend {blendInTime=0.2, blendOutTime=0.5}
    boneToModify {boneName=dyn_pony_3_c}
    rotation {roll=35}
    scale {z=0}
   }
   n3 = "Transform (Modify) Bone - Bone: dyn_pony_6_c"!(ComponentPose=@n2, bAlphaBoolEnabled:bool=True, Translation="0,0,0", Rotation="0,0,40") -> (Pose) {
    alphaInputType=Bool
    rotationMode=BMM_Additive
    translationMode=BMM_Additive
    alphaBoolBlend {blendInTime=0.5, blendOutTime=0.5}
    boneToModify {boneName=dyn_pony_6_c}
    rotation {roll=40}
    scale {z=0}
   }
   n4 = Anim Dynamics - Bone: dyn_pony_2_c!(ComponentPose=Local To Component(LocalPose=Use cached pose 'InputCache'()), Alpha=1, GravityOverride:Vector="-980,0,0") -> (Pose) {
    angularDampingOverride=1
    angularSpringConstant=100
    bAngularSpring=true
    bOverrideAngularDamping=true
    bUseGravityOverride=true
    gravityScale=0.35
    simulationSpace=RootRelative
    boundBone {boneName=dyn_pony_2_c}
    componentAppliedLinearAccClamp=<6000 6000 6000>
    componentLinearAccScale=<0.5 0.5 0.5>
    componentLinearVelScale=<0.6 0.6 0.6>
    gravityOverride {x=-980}
    physicsBodyDefinitions [{
      boundBone {boneName=dyn_pony_2_c}
      boxExtents=<10 10 10>
      constraintSetup {angularLimitsMax=<0 2 7.5>, angularLimitsMin=<0 -13 -7.5>, angularTarget=<1 0 0>}
      localJointOffset=<4 0 0>
     }]
    relativeSpaceBone {boneName=neck_01}
    simSpaceSettings {maxAngularAcceleration=15, maxAngularVelocity=7, simSpaceAngularAlpha=0.75}
   }
   n5 = Anim Dynamics - Bone: dyn_pony_3_c!(ComponentPose=@n4, Alpha=0.85, GravityOverride:Vector="-980,0,0") -> (Pose) {
    alpha=0.85
    angularDampingOverride=1
    angularSpringConstant=100
    bAngularSpring=true
    bOverrideAngularDamping=true
    bUseGravityOverride=true
    gravityScale=0.35
    simulationSpace=RootRelative
    boundBone {boneName=dyn_pony_3_c}
    componentAppliedLinearAccClamp=<6000 6000 6000>
    componentLinearAccScale=<0.5 0.5 0.5>
    componentLinearVelScale=<0.6 0.6 0.6>
    gravityOverride {x=-980}
    physicsBodyDefinitions [{
      boundBone {boneName=dyn_pony_3_c}
      boxExtents=<10 10 10>
      constraintSetup {angularLimitsMax=<5 2 5>, angularLimitsMin=<-5 -8 -5>, angularTarget=<1 0 0>}
      localJointOffset=<4 0 0>
     }]
    relativeSpaceBone {boneName=neck_01}
    simSpaceSettings {maxAngularAcceleration=15, maxAngularVelocity=7, simSpaceAngularAlpha=0.75}
   }
   n6 = Anim Dynamics - Bone: dyn_pony_4_c!(ComponentPose=@n5, Alpha=0.85, GravityOverride:Vector="-980,0,0") -> (Pose) {
    alpha=0.85
    angularDampingOverride=1
    angularSpringConstant=100
    bAngularSpring=true
    bOverrideAngularDamping=true
    bUseGravityOverride=true
    gravityScale=0.35
    simulationSpace=RootRelative
    boundBone {boneName=dyn_pony_4_c}
    componentAppliedLinearAccClamp=<6000 6000 6000>
    componentLinearAccScale=<0.5 0.5 0.5>
    componentLinearVelScale=<0.6 0.6 0.6>
    gravityOverride {x=-980}
    physicsBodyDefinitions [{
      boundBone {boneName=dyn_pony_4_c}
      boxExtents=<10 10 10>
      constraintSetup {angularLimitsMax=<5 2 5>, angularLimitsMin=<-5 -8 -5>, angularTarget=<1 0 0>}
      localJointOffset=<4 0 0>
     }]
    relativeSpaceBone {boneName=neck_01}
    simSpaceSettings {maxAngularAcceleration=15, maxAngularVelocity=7, simSpaceAngularAlpha=0.75}
   }
   n7 = Anim Dynamics - Bone: dyn_pony_5_c!(ComponentPose=@n6, Alpha=0.85, GravityOverride:Vector="-980,0,0") -> (Pose) {
    alpha=0.85
    angularDampingOverride=1
    angularSpringConstant=100
    bAngularSpring=true
    bOverrideAngularDamping=true
    bUseGravityOverride=true
    gravityScale=0.35
    simulationSpace=RootRelative
    boundBone {boneName=dyn_pony_5_c}
    componentAppliedLinearAccClamp=<6000 6000 6000>
    componentLinearAccScale=<0.5 0.5 0.5>
    componentLinearVelScale=<0.6 0.6 0.6>
    gravityOverride {x=-980}
    physicsBodyDefinitions [{
      boundBone {boneName=dyn_pony_5_c}
      boxExtents=<10 10 10>
      constraintSetup {angularLimitsMax=<5 2 5>, angularLimitsMin=<-5 -8 -5>, angularTarget=<1 0 0>}
      localJointOffset=<4 0 0>
     }]
    relativeSpaceBone {boneName=neck_01}
    simSpaceSettings {maxAngularAcceleration=15, maxAngularVelocity=7, simSpaceAngularAlpha=0.75}
   }
   n8 = Anim Dynamics - Bone: dyn_pony_6_c!(ComponentPose=@n7, Alpha=0.85, GravityOverride:Vector="-980,0,0") -> (Pose) {
    alpha=0.85
    angularDampingOverride=1
    angularSpringConstant=100
    bAngularSpring=true
    bOverrideAngularDamping=true
    bUseGravityOverride=true
    gravityScale=0.35
    simulationSpace=RootRelative
    boundBone {boneName=dyn_pony_6_c}
    componentAppliedLinearAccClamp=<6000 6000 6000>
    componentLinearAccScale=<0.5 0.5 0.5>
    componentLinearVelScale=<0.6 0.6 0.6>
    gravityOverride {x=-980}
    physicsBodyDefinitions [{
      boundBone {boneName=dyn_pony_6_c}
      boxExtents=<10 10 10>
      constraintSetup {angularLimitsMax=<5 2 5>, angularLimitsMin=<-5 -8 -5>, angularTarget=<1 0 0>}
      localJointOffset=<4 0 0>
     }]
    relativeSpaceBone {boneName=neck_01}
    simSpaceSettings {maxAngularAcceleration=15, maxAngularVelocity=7, simSpaceAngularAlpha=0.75}
   }
   n9 = Anim Dynamics - Bone: dyn_pony_7_c!(ComponentPose=@n8, Alpha=0.85, GravityOverride:Vector="-980,0,0") -> (Pose) {
    alpha=0.85
    angularDampingOverride=1
    angularSpringConstant=100
    bAngularSpring=true
    bOverrideAngularDamping=true
    bUseGravityOverride=true
    gravityScale=0.35
    simulationSpace=RootRelative
    boundBone {boneName=dyn_pony_7_c}
    componentAppliedLinearAccClamp=<6000 6000 6000>
    componentLinearAccScale=<0.5 0.5 0.5>
    componentLinearVelScale=<0.6 0.6 0.6>
    gravityOverride {x=-980}
    physicsBodyDefinitions [{
      boundBone {boneName=dyn_pony_7_c}
      boxExtents=<10 10 10>
      constraintSetup {angularLimitsMax=<5 2 5>, angularLimitsMin=<-5 -8 -5>, angularTarget=<1 0 0>}
      localJointOffset=<4 0 0>
     }]
    relativeSpaceBone {boneName=neck_01}
    simSpaceSettings {maxAngularAcceleration=15, maxAngularVelocity=7, simSpaceAngularAlpha=0.75}
   }
   n10 = Anim Dynamics - Bone: dyn_pony_8_c!(ComponentPose=@n9, Alpha=0.85, GravityOverride:Vector="-980,0,0") -> (Pose) {
    alpha=0.85
    angularDampingOverride=1
    angularSpringConstant=100
    bAngularSpring=true
    bOverrideAngularDamping=true
    bUseGravityOverride=true
    gravityScale=0.35
    simulationSpace=RootRelative
    boundBone {boneName=dyn_pony_8_c}
    componentAppliedLinearAccClamp=<6000 6000 6000>
    componentLinearAccScale=<0.5 0.5 0.5>
    componentLinearVelScale=<0.6 0.6 0.6>
    gravityOverride {x=-980}
    physicsBodyDefinitions [{
      boundBone {boneName=dyn_pony_8_c}
      boxExtents=<10 10 10>
      constraintSetup {angularLimitsMax=<5 2 5>, angularLimitsMin=<-5 -8 -5>, angularTarget=<1 0 0>}
      localJointOffset=<4 0 0>
     }]
    relativeSpaceBone {boneName=neck_01}
    simSpaceSettings {maxAngularAcceleration=15, maxAngularVelocity=7, simSpaceAngularAlpha=0.75}
   }
   n11 = Anim Dynamics - Bone: dyn_pony_9_c!(ComponentPose=@n10, Alpha=0.85, GravityOverride:Vector="-980,0,0") -> (Pose) {
    alpha=0.85
    angularDampingOverride=1
    angularSpringConstant=100
    bAngularSpring=true
    bOverrideAngularDamping=true
    bUseGravityOverride=true
    gravityScale=0.35
    simulationSpace=RootRelative
    boundBone {boneName=dyn_pony_9_c}
    componentAppliedLinearAccClamp=<6000 6000 6000>
    componentLinearAccScale=<0.5 0.5 0.5>
    componentLinearVelScale=<0.6 0.6 0.6>
    gravityOverride {x=-980}
    physicsBodyDefinitions [{
      boundBone {boneName=dyn_pony_9_c}
      boxExtents=<10 10 10>
      constraintSetup {angularLimitsMax=<5 2 5>, angularLimitsMin=<-5 -8 -5>, angularTarget=<1 0 0>}
      localJointOffset=<4 0 0>
     }]
    relativeSpaceBone {boneName=neck_01}
    simSpaceSettings {maxAngularAcceleration=15, maxAngularVelocity=7, simSpaceAngularAlpha=0.75}
   }
   n12 = Blend Poses by bool(BlendPose_0=Component To Local(ComponentPose=@n11), BlendPose_1=Component To Local(ComponentPose=@n3)) {blendType=Linear}
   n13 = Anim Dynamics - Bone: dyn_pony_3_c!(
    ComponentPose=Local To Component(LocalPose=Use cached pose 'InputCache'())
    Alpha=1
    GravityOverride="0,0,-980"
    ExternalForce:Vector="0,0,0"
   ) -> (Pose) {
    angularDampingOverride=1
    angularSpringConstant=6500
    bAngularSpring=true
    bGravityOverrideInSimSpace=true
    bOverrideAngularDamping=true
    bOverrideLinearDamping=true
    bUseGravityOverride=true
    gravityScale=0.35
    linearDampingOverride=1
    simulationSpace=RootRelative
    alphaBoolBlend {blendOption=Linear}
    alphaScaleBiasClamp {bMapRange=true, outRange {max=0.5}}
    boundBone {boneName=dyn_pony_3_c}
    componentAppliedLinearAccClamp=<3000 3000 3000>
    componentLinearAccScale=<0.5 1 1>
    componentLinearVelScale=<0.5 1 1>
    gravityOverride {z=-980}
    physicsBodyDefinitions [{
      sphereCollisionRadius=0
      boundBone {boneName=dyn_pony_3_c}
      boxExtents=<30 30 30>
      constraintSetup {angularLimitsMax=<5 0 10>, angularLimitsMin=<-5 -2 -10>, angularTarget=<1 0 0>}
      localJointOffset=<13 0 0>
     }]
    relativeSpaceBone {boneName=pelvis}
   }
   n14 = Anim Dynamics - Bone: dyn_pony_4_c!(ComponentPose=@n13, Alpha=1, GravityOverride="0,0,-980", ExternalForce:Vector="0,0,0") -> (Pose) {
    angularDampingOverride=1
    angularSpringConstant=6500
    bAngularSpring=true
    bGravityOverrideInSimSpace=true
    bOverrideAngularDamping=true
    bOverrideLinearDamping=true
    bUseGravityOverride=true
    gravityScale=0.35
    linearDampingOverride=1
    simulationSpace=RootRelative
    alphaBoolBlend {blendOption=Linear}
    alphaScaleBiasClamp {bMapRange=true, outRange {max=0.5}}
    boundBone {boneName=dyn_pony_4_c}
    componentAppliedLinearAccClamp=<3000 3000 3000>
    componentLinearAccScale=<0.5 1 1>
    componentLinearVelScale=<0.5 1 1>
    gravityOverride {z=-980}
    physicsBodyDefinitions [{
      sphereCollisionRadius=0
      boundBone {boneName=dyn_pony_4_c}
      boxExtents=<30 30 30>
      constraintSetup {angularLimitsMax=<5 0 10>, angularLimitsMin=<-5 -2 -10>, angularTarget=<1 0 0>}
      localJointOffset=<13 0 0>
     }]
    relativeSpaceBone {boneName=pelvis}
   }
   n15 = Anim Dynamics - Bone: dyn_pony_5_c!(ComponentPose=@n14, Alpha=1, GravityOverride="0,0,-980", ExternalForce:Vector="0,0,0") -> (Pose) {
    angularDampingOverride=1
    angularSpringConstant=6500
    bAngularSpring=true
    bGravityOverrideInSimSpace=true
    bOverrideAngularDamping=true
    bOverrideLinearDamping=true
    bUseGravityOverride=true
    gravityScale=0.35
    linearDampingOverride=1
    simulationSpace=RootRelative
    alphaBoolBlend {blendOption=Linear}
    alphaScaleBiasClamp {bMapRange=true, outRange {max=0.5}}
    boundBone {boneName=dyn_pony_5_c}
    componentAppliedLinearAccClamp=<3000 3000 3000>
    componentLinearAccScale=<0.5 1 1>
    componentLinearVelScale=<0.5 1 1>
    gravityOverride {z=-980}
    physicsBodyDefinitions [{
      sphereCollisionRadius=0
      boundBone {boneName=dyn_pony_5_c}
      boxExtents=<30 30 30>
      constraintSetup {angularLimitsMax=<5 0 10>, angularLimitsMin=<-5 -2 -10>, angularTarget=<1 0 0>}
      localJointOffset=<13 0 0>
     }]
    relativeSpaceBone {boneName=pelvis}
   }
   n16 = Anim Dynamics - Bone: dyn_pony_6_c!(ComponentPose=@n15, Alpha=1, GravityOverride="0,0,-980", ExternalForce:Vector="0,0,0") -> (Pose) {
    angularDampingOverride=1
    angularSpringConstant=6500
    bAngularSpring=true
    bGravityOverrideInSimSpace=true
    bOverrideAngularDamping=true
    bOverrideLinearDamping=true
    bUseGravityOverride=true
    gravityScale=0.35
    linearDampingOverride=1
    simulationSpace=RootRelative
    alphaBoolBlend {blendOption=Linear}
    alphaScaleBiasClamp {bMapRange=true, outRange {max=0.5}}
    boundBone {boneName=dyn_pony_6_c}
    componentAppliedLinearAccClamp=<3000 3000 3000>
    componentLinearAccScale=<0.5 1 1>
    componentLinearVelScale=<0.5 1 1>
    gravityOverride {z=-980}
    physicsBodyDefinitions [{
      sphereCollisionRadius=0
      boundBone {boneName=dyn_pony_6_c}
      boxExtents=<30 30 30>
      constraintSetup {angularLimitsMax=<5 0 10>, angularLimitsMin=<-5 -2 -10>, angularTarget=<1 0 0>}
      localJointOffset=<13 0 0>
     }]
    relativeSpaceBone {boneName=pelvis}
   }
   n17 = Component To Local(ComponentPose=@n16)
   n18 = Get RBanActive() -> (RBanActive:bool)
   n19 = "Transform (Modify) Bone - Bone: dyn_pony_2_c"!(
    ComponentPose=Local To Component(LocalPose=Blend Poses by bool(BlendPose_0=@n17, BlendPose_1=@n12, bActiveValue=@n18) {blendType=Linear})
    bAlphaBoolEnabled:bool=True
    Translation="0,-2,0"
    Rotation="0.000,0.000,10.000"
   ) -> (Pose) {
    alphaInputType=Bool
    rotationMode=BMM_Additive
    translationMode=BMM_Additive
    boneToModify {boneName=dyn_pony_2_c}
    rotation {roll=10}
    scale {z=0}
    translation {y=-2}
   }
   n20 = "Transform (Modify) Bone - Bone: dyn_pony_3_c"(ComponentPose=@n19, bAlphaBoolEnabled=True, Translation="0.000,-2.500,-0.000", Rotation="0.000,0.000,5.000") {
    alphaInputType=Bool
    rotationMode=BMM_Additive
    translationMode=BMM_Additive
    boneToModify {boneName=dyn_pony_3_c}
    rotation {roll=5}
    translation {y=-2.5}
   }
   Output Pose(Result=Component To Local(ComponentPose=@n20)) {layerGroup=None}
   n21 = Control Rig!(Source=Copy Pose From Mesh() {bUseAttachedParent=true}, Alpha, pub isRigidBodies Enabled:bool=false, isWearingBackbling:bool=false) -> (Pose) {
    alphaBoolBlend {blendOption=Linear}
    controlRigAssetReference {blueprintRigClass=&/F_MED_EagerUnifyHelmet_FaceAcc_CtrlRig.F_MED_EagerUnifyHelmet_FaceAcc_CtrlRig_C}
   }
   Save cached pose 'InputCache'(Pose=@n21) {cacheName=InputCache}
  ]}
 EventGraph {nodes [
   Event BlueprintUpdateAnimation()
   > n1 = Sequence()
   > n2 = "Gravity Override State Function (Parameter Asset) (Nativized)"(
    CharacterPartAnimInstance=Self-Reference()
    GameTimeInSeconds=GetGameTimeInSeconds()
    Parameters=/BRCosmetics/Characters/Player/Female/Medium/Bodies/F_MED_EagerUnify/Meshes/Parts/F_MED_EagerUnify_FaceAcc_Parameters
   )
   > Set GravityOverride(GravityOverride=@n2)
   Set RBanActive(execute=@n1.then_1, RBanActive:bool=OR Boolean(A=Get bIsDBNO(), B=Get isWearingBackbling())) -> (Output_Get:bool)
   n3 = Get bIsPlayingEmote()
   > Set isIdleFrontEnd(isIdleFrontEnd:bool=AND Boolean(A=Get bIsInFrontEnd(), B=InvertBool!(bValue:bool=@n3) -> (ReturnValue:bool))) -> (Output_Get:bool)
   Event OnCharacterCustomizationLoaded()
   n4 = CheckForBackbling()
   > n5 = Set isWearingCape(isWearingCape=@n4.bIsBackblingACape)
   > n6 = Branch(Condition=@n5)
   > Set isWearingBackbling(isWearingBackbling=false)
   Set isWearingBackbling(execute=@n6.else, isWearingBackbling=OR Boolean(A=@n4, B=@n4.bIsBackblingLong))
  ]}
]
variables [
 Gravity Override Params {
  type=FortGravityOverrideParameterSet
  default=(Skydive_GoingUp=(GravityOverride=()),Skydive_FloatingDown=(GravityOverride=()),Skydive_DivingDown=(GravityOverride=()),Skydive_Parachuting=(GravityOverride=()),OnGround_Standing=(GravityOverride=()),OnGround_Moving=(GravityOverride=()),OnGround_Moving_HighRotationRate=(GravityOverride=()),OnGround_Crouch_Moving=(GravityOverride=()),EmoteOrMelee=(GravityOverride=()),Swimming=(GravityOverride=()),Front_End=(GravityOverride=()),DBNO=(GravityOverride=()),DBNO_Carried=(GravityOverride=()),OnGround_Falling=(GravityOverride=()),Ziplining=(GravityOverride=()),Skydive_Surfing=(GravityOverride=()),Riding_Standing=(GravityOverride=()),Riding_Moving=(GravityOverride=()),Riding_Moving_Targeting=(GravityOverride=()),Tactical_Sprint=(GravityOverride=()),Rail_Grinding=(GravityOverride=()),)
 }
 GravityOverride {type=Vector}
 GravityOverrideParamsOld {type=GravityOverrideParamsStruct, default {onGround_Moving {gravityOverride=<0 0 -700>}}}
 isIdleFrontEnd {type=boolean}
 isWearingBackbling {type=boolean}
 isWearingCape {type=boolean}
 RBanActive {type=boolean}
]
